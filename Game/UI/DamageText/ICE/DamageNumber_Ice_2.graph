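# Translucent unlit material for ice-style floating damage numbers, drawing a misty glyph layer.
# A UV2 mask switches between digit and plus/minus textures, both offset by flow-map distortion borrowed from the Fire set.
# Glyph opacity combines depth fades, a soft sphere mask and panning noise.
# The shared DamageNumberWPO function takes normals, scale and offsets, and its time fades the glyph, tinted by desaturated sun and sky light.
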Material /Game/UI/DamageText/ICE/DamageNumber_Ice_2
graph {
 properties {
  bEnableResponsiveAA=true
  blendMode=BLEND_Translucent
  numCustomizedUVs=6
  pixelDepthOffsetMode=PDOM_Legacy
  previewMesh=/Game/UI/DamageText/Default/standard_DamageNumbers8
  shadingModel=MSM_Unlit
  translucencyLightingMode=TLM_Surface
  usedShadingModels=Unlit
 }
 nodes [
  AppendVector()
  Constant()
  Constant() {r=0.5}
  Constant() {r=-5.1293}
  MaterialFunctionCall() {materialFunction=&TimeWithSpeedVariable}
  n1 = TextureCoordinate() {coordinateIndex=1}
  MaterialFunctionCall!(
   Vector2=@n1
   MaximumNumberOfDigits
   UVs=Panner() {bFractionalPart=true, speedY=0.5}
   DebugTextLocation RG_UpperRight BA_LowerLeft
   Component Spacing
  ) {materialFunction=/Engine/Functions/Engine_MaterialFunctions02/Utility/DebugFloat2Values}
  n2 = SphereMask(A=TextureCoordinate(), B=Constant() {r=0.5}) {attenuationRadius=0.5, hardnessPercent=0}
  n3 = TextureSample(UVs=@n1) {
   desc=DamageNumbers_Ice_RmistCore_G_Edges_B_AO
   samplerType=SAMPLERTYPE_LinearColor
   texture=/Game/UI/DamageText/Fire/DamageNumbers_Fire_RG_FlowMap
  }
  n4 = TextureSample(UVs=TextureCoordinate() {vTiling=0.1}) {
   desc=DamageNumbers_Ice_RmistCore_G_Edges_B_AO
   samplerType=SAMPLERTYPE_LinearColor
   texture=&DamageNumbers_Fire_RG_FlowMap
  }
  n5 = ComponentMask(None=TextureCoordinate() {coordinateIndex=2}) {r=true}
  n6 = Subtract(A=ComponentMask(None=LinearInterpolate(A=@n3.RGB, B=@n4.RGB, Alpha=@n5) {constB=0}) {g=true, r=true}) {constB=0.5}
  n7 = ScalarParameter() {defaultValue=0.6814, parameterName=Letter Distortion Intensity}
  n8 = MaterialFunctionCall!(
   Diffuse
   Normal Map
   "Flow Vector Map (see tooltip)"=Multiply(A=Multiply(A=@n6, B=@n7), B=Constant2Vector() {g=-1, r=1})
   UV texture 2 offset=Constant()
   "Time (see tooltip)"
   useMipLevel
   UVs=@n1
   Texture Size
   Texture Mip Bias
  ) {materialFunction=/Engine/Functions/Engine_MaterialFunctions02/Texturing/Flowmaps.FlowMaps}
  n9 = Subtract(A=@n8.UVs, B=@n1)
  n10 = Add(A=TextureCoordinate() {vTiling=0.5}, B=AppendVector(A=Constant(), B=ScalarParameter() {parameterName=+Or-}))
  n11 = LinearInterpolate(
   A=TextureSample(UVs=@n8.UVs) {samplerType=SAMPLERTYPE_LinearColor, texture=&/DamageNumbers_Ice_RhMap_G_Mist_B_Alpha}.G
   B=TextureSample(UVs=Add(A=@n9, B=@n10)) {samplerType=SAMPLERTYPE_LinearColor, texture=&/DamageNumbers_Ice_Hard_RhMap_G_Mist_B_Alpha}.A
   Alpha=@n5
  ) {constB=0}
  n12 = DepthFade(Opacity=Multiply(A=DepthFade(Opacity=@n11), B=ScalarParameter() {defaultValue=0.2239, parameterName=OpacityMultiplier}) {constB=0.2965}) {fadeDistanceDefault=200}
  n13 = TextureCoordinate()
  n14 = Multiply(
   A=Constant2Vector() {g=1}
   B=MaterialFunctionCall(Float2=TextureCoordinate() {coordinateIndex=5}) {materialFunction=&BreakOutFloat2Components}.R
  )
  n15 = TextureSample(UVs=Panner(Coordinate=Add(A=@n13, B=@n14)) {speedY=-0.1}) {samplerType=SAMPLERTYPE_Grayscale, texture=&LowResBlurredNoise}
  n16 = ScalarParameter() {defaultValue=0.6637, parameterName=MistLowEndOpacityLevel}
  Multiply(A=Multiply(A=Multiply(A=@n2, B=@n12), B=LinearInterpolate(A=@n15.R, B=@n16) {constAlpha=0.75}) {desc=o}) {constB=0.4681}
  n17 = CollectionParameter() {
   collection=&FortniteMaterialParameters
   parameterName=SkyLightColorAndIntensity
  }
  n18 = LinearInterpolate(
   A=ComponentMask(None=CollectionParameter() {collection=&FortniteMaterialParameters, parameterName=SunLightColor}) {b=true, g=true, r=true}
   B=ComponentMask(None=@n17) {b=true, g=true, r=true}
  )
  n19 = Constant() {r=0.5}
  n20 = TextureSampleParameter2D(UVs=@n1) {group=Visuals, parameterName=Normal, samplerType=SAMPLERTYPE_Normal, texture=&/DamageNumbers_Ice_Normals}
  n21 = TextureSample(UVs=@n10) {samplerType=SAMPLERTYPE_Normal, texture=/Game/UI/DamageText/Default/+-Normal.+-Normal}
  n22 = Normalize(VectorInput=Multiply(A=LinearInterpolate(A=@n20.RGB, B=@n21.RGB, Alpha=@n5), B=Constant3Vector() {constant=<rgba 1 1 0.5 0>}))
  n23 = ScalarParameter() {defaultValue=2.0095, parameterName=Additional Scale}
  n24 = ScalarParameter() {defaultValue=-25.753, parameterName=Additional Camera Offset}
  n25 = Constant3Vector() {constant=<rgba 0 0 1 0>}
  n26 = ComponentMask(None=Multiply(
    A=ScalarParameter() {defaultValue=-3.6941, parameterName=YOffset}
    B=Transform(None=@n25) {transformSourceType=TRANSFORMSOURCE_World, transformType=TRANSFORM_View}
   )) {g=true, r=true}
  Multiply(
   A=Multiply(A=@n11, B=Desaturation(None=@n18, Fraction=@n19))
   B=OneMinus(None=MaterialFunctionCall!(
     Normals=@n22
     Debug texture coordinate input
     Animation
     Additonal Scale=@n23
     "Additional Camera Offset "=@n24
     "Additonal Screenspace XY offset "=@n26
    ) {materialFunction=/Game/UI/DamageText/Base/DamageNumberWPO}.Time)
  )
  ScalarParameter() {defaultValue=0.43, parameterName=debug animation}
 ]
}
