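# Masked, unlit material for ice-styled floating damage numbers, built around the shared DamageNumberWPO function for animation.
# Packed mask and normal textures blend with "+/-" symbol variants via a UV channel, shaping a border-to-ice color mix with depth-based border darkening.
# It fakes lighting from global sun and sky parameters, a cubemap reflection and time-of-day dimming.
# The function's letter highlight, critical-hit and time outputs then scale brightness.
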Material /Game/UI/DamageText/ICE/FlyingTextMat_IceMat
graph {
 properties {
  bEnableResponsiveAA=true
  blendMode=BLEND_Masked
  numCustomizedUVs=4
  pixelDepthOffsetMode=PDOM_Legacy
  previewMesh=/Game/UI/DamageText/Default/standard_DamageNumbers8
  shadingModel=MSM_Unlit
  translucencyLightingMode=TLM_Surface
  usedShadingModels=Unlit
 }
 nodes [
  Constant() {r=0.5}
  FeatureLevelSwitch()
  n1 = VectorParameter() {parameterName=IceColor, defaultValue=<rgba 0.2316 0.3622 0.965 1>}
  n2 = TextureCoordinate() {coordinateIndex=1}
  n3 = TextureSample(UVs=@n2) {
   desc=DamageNumbers_Ice_RmistCore_G_Edges_B_AO
   samplerType=SAMPLERTYPE_LinearColor
   texture=&/DamageNumbers_Ice_RmistCore_G_Edges_B_AO
  }
  n4 = Add(A=TextureCoordinate() {vTiling=0.5}, B=AppendVector(A=Constant(), B=ScalarParameter() {parameterName=+Or-}))
  n5 = TextureSample(UVs=@n4) {
   samplerType=SAMPLERTYPE_LinearColor
   texture=&/DamageNumbers_Ice_RmistCore_G_Edges_B_AO_Hard+-.DamageNumbers_Ice_RmistCore_G_Edges_B_AO_Hard+-
  }
  n6 = ComponentMask(None=TextureCoordinate() {coordinateIndex=2}) {r=true}
  n7 = MaterialFunctionCall(Float3=LinearInterpolate(A=@n3.RGB, B=@n5.RGB, Alpha=@n6)) {
   desc=DamageNumbers_Ice_RmistCore_G_Edges_B_AO
   materialFunction=&BreakOutFloat3Components
  }
  n8 = TextureSample(UVs=@n2) {
   desc=DamageNumbers_Ice_RhMap_G_Mist_B_Alpha
   samplerType=SAMPLERTYPE_LinearColor
   texture=&/DamageNumbers_Ice_RhMap_G_Mist_B_Alpha
  }
  n9 = MaterialFunctionCall(Float3=LinearInterpolate(
    A=@n8.RGB
    B=TextureSample(UVs=@n4) {samplerType=SAMPLERTYPE_LinearColor, texture=&/DamageNumbers_Ice_Hard_RhMap_G_Mist_B_Alpha}.RGB
    Alpha=@n6
   )) {desc=DamageNumbers_Ice_RhMap_G_Mist_B_Alpha, materialFunction=&BreakOutFloat3Components}
  n10 = Multiply(A=@n7.R, B=@n9.B)
  Multiply(A=@n1.RGB, B=@n10)
  n11 = TextureSampleParameter2D(UVs=@n2) {group=Visuals, parameterName=Normal, samplerType=SAMPLERTYPE_Normal, texture=&/DamageNumbers_Ice_Normals}
  n12 = LinearInterpolate(A=@n11.RGB, B=TextureSample(UVs=@n4) {samplerType=SAMPLERTYPE_Normal, texture=&/Ice_-_hardNormalsMap}.RGB, Alpha=@n6)
  n13 = Constant3Vector() {constant=<rgba 1 1 0.5 0>}
  n14 = MaterialFunctionCall!(
   Normals=Normalize(VectorInput=Multiply(A=@n12, B=@n13))
   Debug texture coordinate input
   Animation
   Additonal Scale
   "Additional Camera Offset "
   "Additonal Screenspace XY offset "
  ) {materialFunction=/Game/UI/DamageText/Base/DamageNumberWPO}
  n15 = Constant() {r=1}
  n16 = CollectionParameter() {
   collection=&FortniteMaterialParameters
   parameterName=SunlightDirectionalVector
  }
  n17 = DotProduct(A=Transform(None=@n14.Normals), B=ComponentMask(None=@n16) {b=true, g=true, r=true})
  n18 = Multiply(A=OneMinus(None=@n7.G), B=@n9.R)
  n19 = ComponentMask(None=CollectionParameter() {collection=&FortniteMaterialParameters, parameterName=SunLightColor}) {b=true, g=true, r=true}
  n20 = Multiply(
   A=LinearInterpolate(A=ConstantBiasScale(None=@n17) {bias=1.5}, B=Clamp(None=@n17), Alpha=@n18)
   B=Desaturation(None=Normalize(VectorInput=@n19), Fraction=Constant() {r=0.5})
  )
  n21 = VectorParameter() {parameterName=Border Color, defaultValue=<rgba 0.0281 0.0755 0.26 0.6195>}
  n22 = PixelDepth()
  n23 = Multiply(A=Clamp(None=Divide(A=@n22.R) {constB=2048}), B=@n21.A) {constB=0.8, desc=min brightness}
  n24 = LinearInterpolate(A=Multiply(A=@n21.RGB, B=OneMinus(None=@n23)), B=Add(A=@n1.RGB, B=@n18), Alpha=@n10)
  n25 = CollectionParameter() {
   collection=&FortniteMaterialParameters
   parameterName=SkyLightColorAndIntensity
  }
  n26 = ReflectionVectorWS(CustomWorldNormal=@n14.Normals)
  n27 = Constant() {r=100}
  n28 = Multiply(
   A=ComponentMask(None=@n25) {b=true, g=true, r=true}
   B=Multiply(
    A=TextureSampleParameterCube!(UVs=@n26, Bias=@n27, Apply View MipBias) {mipValueMode=TMVM_MipBias, parameterName=Cubemap, texture=/Game/Packages/Fortress_Sky/TexturesHDR/T_hdrsky_test03_low}.RGB
    B=Power(Base=@n7.B) {constExponent=1}
   )
  )
  n29 = Multiply(
   A=MaterialFunctionCall(Color_Morning=@n15, Color_Day=@n15, Color_Evening=@n15, Color_Night=Constant() {r=0.615}) {materialFunction=&Color_Time_of_Day}.Result
   B=Add(A=Multiply(A=@n20, B=@n24), B=Multiply(A=@n24, B=@n28))
  )
  n30 = Add(A=@n14.Pixel shader Letter Highlight - uv 3) {constB=0.4}
  Multiply(A=OneMinus(None=@n14.Time), B=Multiply(A=Multiply(A=@n29, B=@n30), B=Add(A=@n14.isCriticalHit?) {constB=0.75}))
  Multiply(A=OneMinus(None=@n9.R)) {constB=5}
  ScalarParameter() {defaultValue=0.43, parameterName=debug animation}
 ]
}
