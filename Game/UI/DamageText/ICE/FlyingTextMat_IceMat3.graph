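# Unlit translucent material for ice-styled flying damage numbers with a melting-ice cover.
# UV channel 2 blends digit and plus/minus normal and mask textures, and the normals feed DamageNumberWPO for animation.
# Lighting is computed in-graph from sun direction and color plus skylight and a cubemap reflection, tinted by Border Color.
# The edge mask combines with Fresnel, critical hits boost mask and brightness, and the lit color fades with animation time.
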Material /Game/UI/DamageText/ICE/FlyingTextMat_IceMat3
graph {
 properties {
  bEnableResponsiveAA=true
  bTangentSpaceNormal=false
  blendMode=BLEND_Translucent
  numCustomizedUVs=4
  pixelDepthOffsetMode=PDOM_Legacy
  previewMesh=/Game/UI/DamageText/Default/standard_DamageNumbers8
  shadingModel=MSM_Unlit
  translucencyLightingMode=TLM_Surface
  usedShadingModels=Unlit
 }
 nodes [
  n1 = TextureCoordinate() {coordinateIndex=1}
  n2 = TextureSampleParameter2D(UVs=@n1) {group=Visuals, parameterName=Normal, samplerType=SAMPLERTYPE_Normal, texture=&/damageNumber_MeltingIce_Normals}
  n3 = Add(A=TextureCoordinate() {vTiling=0.5}, B=AppendVector(A=Constant(), B=ScalarParameter() {parameterName=+Or-}))
  n4 = TextureSample(UVs=@n3) {samplerType=SAMPLERTYPE_Normal, texture=&/damageNumber_MeltingIce_Normals_+-.damageNumber_MeltingIce_Normals_+-}
  n5 = ComponentMask(None=TextureCoordinate() {coordinateIndex=2}) {r=true}
  n6 = LinearInterpolate(A=@n2.RGB, B=@n4.RGB, Alpha=@n5)
  n7 = Multiply(A=@n6, B=Constant3Vector() {constant=<rgba 1 1 0.5 0>})
  n8 = Normalize(VectorInput=@n7)
  n9 = MaterialFunctionCall!(
   Normals=@n8
   Debug texture coordinate input
   Animation
   Additonal Scale=ScalarParameter() {defaultValue=0.9558, parameterName=Scale additional}
   "Additional Camera Offset "=ScalarParameter() {defaultValue=0.3009, parameterName=Second Offset}
   "Additonal Screenspace XY offset "
  ) {materialFunction=/Game/UI/DamageText/Base/DamageNumberWPO}
  n10 = Transform(None=@n9.Normals)
  n11 = CollectionParameter() {
   collection=&FortniteMaterialParameters
   parameterName=SunlightDirectionalVector
  }
  n12 = DotProduct(A=@n10, B=ComponentMask(None=@n11) {b=true, g=true, r=true})
  Abs(None=Clamp(None=@n12) {maxDefault=0, minDefault=-1})
  Constant() {r=0.5}
  DotProduct()
  MaterialFunctionCall!(Center Location, 0-1 UVs, XY scale, OverridePivot, "Normals (Optional)"=@n8) {materialFunction=&Sprite}
  n13 = TextureSample(UVs=@n1) {
   desc=DamageNumbers_Ice_RmistCore_G_Edges_B_AO
   samplerType=SAMPLERTYPE_LinearColor
   texture=&/Damage_Text_IceCover_Masks
  }
  n14 = MaterialFunctionCall(Float3=LinearInterpolate(
    A=@n13.RGB
    B=TextureSample(UVs=@n3) {samplerType=SAMPLERTYPE_LinearColor, texture=&/Damage_Text_IceCover_Masks_+-.Damage_Text_IceCover_Masks_+-}.RGB
    Alpha=@n5
   )) {
   desc=DamageNumbers_Ice_RmistCore_G_Edges_B_AO
   materialFunction=&BreakOutFloat3Components
  }
  n15 = Multiply(A=Fresnel(Normal=@n10) {baseReflectFraction=0.07, exponent=1}, B=ScalarParameter() {defaultValue=1, parameterName=Opacity multiplier})
  n16 = Multiply(A=@n9.isCriticalHit?) {constB=0.5}
  n17 = Multiply(A=@n14.G, B=Clamp(None=Add(A=@n15, B=@n16))) {desc=mask}
  Multiply(A=@n17, B=@n17)
  Multiply(A=ComponentMask(None=TextureCoordinate() {coordinateIndex=3}) {r=true}, B=Constant2Vector() {r=0.2}) {constB=0.2}
  n18 = ConstantBiasScale(None=@n12) {bias=1.5}
  n19 = Clamp(None=@n12)
  n20 = Normalize(VectorInput=ComponentMask(None=CollectionParameter() {collection=&FortniteMaterialParameters, parameterName=SunLightColor}) {b=true, g=true, r=true})
  n21 = Constant() {r=0.5}
  n22 = LinearInterpolate(A=VectorParameter() {parameterName=Border Color, defaultValue=<rgba 0.0281 0.0755 0.26 1>}.RGB)
  n23 = Multiply(A=Multiply(A=LinearInterpolate(A=@n18, B=@n19), B=Desaturation(None=@n20, Fraction=@n21)), B=@n22)
  n24 = CollectionParameter() {
   collection=&FortniteMaterialParameters
   parameterName=SkyLightColorAndIntensity
  }
  n25 = TextureSampleParameterCube!(UVs=ReflectionVectorWS(CustomWorldNormal=@n9.Normals), Bias=Constant() {r=100}, Apply View MipBias) {mipValueMode=TMVM_MipBias, parameterName=Cubemap, texture=/Game/Packages/Fortress_Sky/TexturesHDR/T_hdrsky_test03_low}
  n26 = Multiply(A=@n22, B=Multiply(A=ComponentMask(None=@n24) {b=true, g=true, r=true}, B=Multiply(A=@n25.RGB)))
  Multiply(A=OneMinus(None=@n9.Time), B=Multiply(A=Add(A=@n23, B=@n26), B=Add(A=@n9.isCriticalHit?) {constB=0.5}))
  n27 = Constant3Vector() {constant=<rgba 0 0 1 0>}
  Multiply(A=@n6, B=@n27)
  Normalize(VectorInput=LinearInterpolate(A=@n7, B=@n27))
  ScalarParameter() {defaultValue=0.43, parameterName=debug animation}
  TextureSample() {samplerType=SAMPLERTYPE_Normal, texture=/Game/UI/DamageText/Default/+-Normal.+-Normal}
  Transform(None=LinearInterpolate(A=@n9.Normals, B=@n9.Normals))
  VectorParameter() {group=Visuals, parameterName=Color, defaultValue=<rgba 0.9639 1 0.9798 0.5>}
 ]
}
